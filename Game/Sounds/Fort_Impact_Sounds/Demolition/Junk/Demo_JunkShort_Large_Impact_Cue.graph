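# Sound cue for short, large junk impact hits, filed under demolition impact sounds.
# A random node picks one of three equally weighted impact waves, then a modulator applies slight random pitch variation.
# It plays in the impacts sound class at reduced volume, with demolition-hit attenuation and a concurrency set capping simultaneous instances at four.
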SoundCue /Game/Sounds/Fort_Impact_Sounds/Demolition/Junk/Demo_JunkShort_Large_Impact_Cue
root=Modulator_2
nodes [
 {id=Modulator_2, children [Random_15], properties {pitchMax=1.07, pitchMin=0.92}}
 {id=Random_15, children [WavePlayer_91, WavePlayer_92, WavePlayer_94], properties {weights [1,1,1]}}
 {id=WavePlayer_91, properties {soundWaveAssetPtr=&/Junk_ImpactShort_Large01}}
 {id=WavePlayer_92, properties {soundWaveAssetPtr=&/Junk_ImpactShort_Large02}}
 {id=WavePlayer_94, properties {soundWaveAssetPtr=&/Junk_ImpactShort_Large04}}
]
properties {
 attenuationSettings=/Game/Sounds/Attenuation/Fort_Demolition_Hit_Small_Sized_Object_Att
 maxDistance=4000
 soundClassObject=&Fort_Impacts
 volumeMultiplier=0.68
 concurrencyOverrides {maxCount=4}
 concurrencySet [&CONC_Object_Impacts_FourMax]
}
