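# Sound cue for a short, medium junk impact in the demolition impact sound set.
# A Random node picks one of three equally weighted junk impact waves and feeds a Modulator that slightly varies pitch.
# The cue lowers volume, uses small-object demolition hit attenuation, routes to the impacts sound class, and references a four-voice object impact concurrency set.
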SoundCue /Game/Sounds/Fort_Impact_Sounds/Demolition/Junk/Demo_JunkShort_Medium_Impact_Cue
root=Modulator_2
nodes [
 {id=Modulator_2, children [Random_14], properties {pitchMax=1.07, pitchMin=0.92}}
 {id=Random_14, children [WavePlayer_85, WavePlayer_86, WavePlayer_87], properties {weights [1,1,1]}}
 {id=WavePlayer_85, properties {soundWaveAssetPtr=&/Junk_ImpactShort_Medium01}}
 {id=WavePlayer_86, properties {soundWaveAssetPtr=&/Junk_ImpactShort_Medium02}}
 {id=WavePlayer_87, properties {soundWaveAssetPtr=&/Junk_ImpactShort_Medium04}}
]
properties {
 attenuationSettings=/Game/Sounds/Attenuation/Fort_Demolition_Hit_Small_Sized_Object_Att
 maxDistance=4000
 soundClassObject=&Fort_Impacts
 volumeMultiplier=0.68
 concurrencyOverrides {maxCount=4}
 concurrencySet [&CONC_Object_Impacts_FourMax]
}
